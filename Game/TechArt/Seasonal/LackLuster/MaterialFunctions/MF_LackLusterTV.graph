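# LackLuster seasonal material function that layers animated concentric TV ripples onto the incoming material attributes' normal.
# Pre-skinned X/Z position, interpolated from the vertex stage, is offset, scaled, converted to radial coordinates and panned to sample a ripple normal map.
# A soft sphere mask at the ripple center drives flattening, fully flat at the middle, before angle-corrected blending.
# The "Use TV Normals" static switch otherwise passes input through unchanged.
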MaterialFunction /Game/TechArt/Seasonal/LackLuster/MaterialFunctions/MF_LackLusterTV
graph {nodes [
  n1 = FunctionInput() {inputType=FunctionInput_MaterialAttributes}
  n2 = GetMaterialAttributes(None=@n1) {attributeGetTypes [#1]}
  n3 = VertexInterpolator(VS=ComponentMask(None=PreSkinnedPosition()) {b=true, r=true})
  n4 = AppendVector(
   A=ScalarParameter() {defaultValue=11.5, parameterName=TV Ripple U Offset}
   B=ScalarParameter() {defaultValue=-72, parameterName=TV Ripple V Offset}
  )
  n5 = Divide(A=Add(A=@n3.PS, B=@n4), B=ScalarParameter() {defaultValue=100, parameterName=TV UV Scale})
  n6 = ScalarParameter() {parameterName=TV RipplePanSpeed_U}
  n7 = ScalarParameter() {defaultValue=-0.08, parameterName=TV RipplePanSpeed_V}
  n8 = TextureSampleParameter2D(UVs=Panner(
    Coordinate=MaterialFunctionCall(Vector or UVs=@n5) {materialFunction=&VectorToRadialValue}.Radial Coordinates
    Speed=AppendVector(A=@n6, B=@n7)
   ) {bFractionalPart=true, speedY=-0.03}) {
   parameterName=Concentric Ripples Normal
   samplerType=SAMPLERTYPE_Normal
   texture=/Game/TechArt/Seasonal/LackLuster/Textures/T_Lackluster_Phased_Normal
  }
  n9 = ScalarParameter() {defaultValue=0.85, parameterName=TV Flatten Normals}
  n10 = SphereMask(
   A=Multiply(A=@n4) {constB=-1}
   B=@n3.PS
   Radius=ScalarParameter() {defaultValue=31.3795, parameterName=TV Normal Sphere Mask Radius}
   Hardness=ScalarParameter() {parameterName=TV Normal Sphere Mask Harness}
  ) {hardnessPercent=0}
  n11 = MaterialFunctionCall(
   BaseNormal=@n2.Normal
   AdditionalNormal=MaterialFunctionCall(Normal=@n8.RGB, Flatness=LinearInterpolate(A=@n9, Alpha=@n10)) {materialFunction=&FlattenNormal}.Result
  ) {materialFunction=&BlendAngleCorrectedNormals}
  FunctionOutput(None=StaticSwitchParameter(
    True=SetMaterialAttributes!(MaterialAttributes=@n2.MaterialAttributes, Normal=@n11.Result) {attributeSetTypes [#1]}
    False=@n1
   ) {parameterName=Use TV Normals})
 ]}
