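# Material function that adds the LackLuster seasonal chrome look to landscape material attributes after the runtime virtual texture stage.
# The virtual-texture path hue-shifts colour by Fresnel, adds panning streak-driven curve-atlas reflections, tints by time of day and adjusts metallic and roughness.
# An input mask, optionally multiplied by a cheap slope mask, blends it in.
# Without virtual textures, a cheaper mobile chrome blend replaces it; quality switches trim cost further.
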MaterialFunction /Game/TechArt/Seasonal/LackLuster/MaterialFunctions/MF_LackLuster_Landscape_PostRVT
graph {nodes [
  n1 = NamedRerouteUsage()
  n2 = NamedRerouteUsage()
  n3 = NamedRerouteUsage()
  n4 = VectorParameter() {parameterName=ChromeMobileColorMax, defaultValue=<rgba 0.3686 0.4 0.3319 0>}
  n5 = OneMinus(None=Divide(
    A=VectorParameter() {group=DiffuseSource, parameterName=DiffuseSource_ColorDodgeBase, defaultValue=<rgba 0.05 0.05 0.05 0>}.RGB
    B=NamedRerouteUsage()
   ))
  n6 = Constant() {r=1}
  n7 = Multiply(
   A=Max(A=Min(A=@n3, B=@n4.RGB), B=VectorParameter() {parameterName=ChromeMobileColorMin, defaultValue=<rgba 0.3456 0.2887 0.3872 0>}.RGB)
   B=LinearInterpolate(
    A=VectorParameter() {parameterName=ChromeMobileColorLow, defaultValue=<rgba 0.3184 0.4737 0.5052 0>}.RGB
    B=VectorParameter() {parameterName=ChromeMobileColor, defaultValue=<rgba 0.4884 0.8343 0.9045 0>}.RGB
    Alpha=Desaturation(None=@n5, Fraction=@n6)
   )
  )
  n8 = NamedRerouteUsage()
  n9 = ScalarParameter() {defaultValue=3.5, parameterName=test}
  n10 = SetMaterialAttributes!(
   MaterialAttributes=NamedRerouteUsage()
   BaseColor=Multiply(
    A=MaterialFunctionCall(Hue Shift Percentage=@n2, Texture=@n7) {materialFunction=&HueShift}.Result
    B=Power(Base=@n8, Exp=@n9)
   )
   Metallic=ScalarParameter() {defaultValue=0.7, parameterName=ChromeMetallicMobile}
   Roughness=ScalarParameter() {defaultValue=0.15, parameterName=ChromeRoughnessMobile}
   Specular=ScalarParameter() {defaultValue=0.5, parameterName=ChromeMobileSpecular}
  ) {attributeSetTypes [
    #1
    #2
    #3
    #4
   ]}
  n11 = Multiply(A=NamedRerouteUsage(), B=NamedRerouteUsage())
  n12 = NamedRerouteUsage()
  n13 = MaterialFunctionCall() {materialFunction=&MF_Fresnel_FX}
  n14 = ScalarParameter() {defaultValue=3, group=HueShift, parameterName=FresnelHueShiftScaleMobile}
  n15 = ScalarParameter() {defaultValue=0.3, group=HueShift, parameterName=FresnelHueShiftScale}
  n16 = FunctionInput(Preview=Multiply(A=@n13.Fresnel, B=VirtualTextureFeatureSwitch!(No=@n14, Yes=@n15))) {bUsePreviewValueAsDefault=true, inputName=HueShiftOverride, inputType=FunctionInput_Scalar, sortPriority=4}
  n17 = NamedRerouteUsage()
  n18 = Divide(A=NamedRerouteUsage(), B=ScalarParameter() {defaultValue=2688, parameterName=Chrome_Divisor_UV_a})
  n19 = Add(
   A=Multiply(
    A=ComponentMask(None=@n17) {g=true, r=true}
    B=ScalarParameter() {defaultValue=0.3, group=ReflectionCurve, parameterName=ReflectionCurveUVDistortionScale}
   )
   B=@n18
  )
  n20 = ScalarParameter() {defaultValue=0.012, parameterName=ChromePannerSpeedU}
  n21 = ScalarParameter() {defaultValue=0.0168, parameterName=ChromePannerSpeedV}
  n22 = TextureSample(UVs=Panner(Coordinate=QualitySwitch(Default=@n19, Low=@n18, Medium=@n18), Speed=AppendVector(A=@n20, B=@n21)) {bFractionalPart=true, speedX=0.015, speedY=0.0205}) {samplerType=SAMPLERTYPE_LinearColor, texture=/Game/TechArt/Seasonal/LackLuster/Textures/T_Lackluster_Streaks}
  n23 = CurveAtlasRowParameter(CurveTime=Frac(None=Add(A=@n13.Fresnel, B=@n22.R))) {
   atlas=/Game/TechArt/Seasonal/LackLuster/Curves/CA_Lackluster
   curve=/Game/TechArt/Seasonal/LackLuster/Curves/C_Lackluster_Reflection
   parameterName=ReflectionCurve
  }
  n24 = Multiply(A=NamedRerouteUsage(), B=@n23)
  n25 = NamedRerouteUsage()
  n26 = ScalarParameter() {defaultValue=0.05, parameterName=ReflectionCurveScaleLow}
  n27 = ScalarParameter() {defaultValue=0.7, parameterName=ReflectionCurveScaleHigh}
  n28 = Power(Base=@n13.Fresnel Inverted, Exp=ScalarParameter() {defaultValue=0.4, group=Normals, parameterName=EdgeNormal_FresnelExponent})
  n29 = Add(
   A=MaterialFunctionCall(Hue Shift Percentage=NamedRerouteDeclaration(None=@n16) {name=HueShiftPercentage, nodeColor=<rgba 0 1 0.625 1>}, Texture=NamedRerouteUsage()) {materialFunction=&HueShift}.Result
   B=Multiply(A=StaticSwitch(True=@n24, False=@n23, Value=@n25), B=LinearInterpolate(A=@n26, B=@n27, Alpha=@n28) {constA=0.05})
  ) {constB=0}
  n30 = MaterialFunctionCall(
   Color_Morning=VectorParameter() {parameterName=ChromeLandscapeColorScaleMorning, defaultValue=<rgba 2 1.6063 1.3681 0>}.RGB
   Color_Day=VectorParameter() {parameterName=ChromeLandscapeColorScaleDay, defaultValue=<rgba 1.023 1.1216 1.2 0>}.RGB
   Color_Evening=VectorParameter() {parameterName=ChromeLandscapeColorScaleEvening, defaultValue=<rgba 1.8 1.7039 1.6859 0>}.RGB
   Color_Night=VectorParameter() {parameterName=ChromeLandscapeColorScaleNight, defaultValue=<rgba 1.7 1.4424 1.0377 0>}.RGB
  ) {materialFunction=&Color_Time_of_Day}
  n31 = ScalarParameter() {defaultValue=1, parameterName=SpecLightMask_Low}
  n32 = ScalarParameter() {defaultValue=0.05, parameterName=SpecLightMask_High}
  n33 = NamedRerouteUsage()
  n34 = ScalarParameter() {defaultValue=0.725, parameterName=Metallic}
  n35 = NamedRerouteUsage()
  n36 = ScalarParameter() {defaultValue=0.35, parameterName=RoughnessSpecRemoval}
  n37 = SetMaterialAttributes!(
   MaterialAttributes=NamedRerouteUsage()
   BaseColor=Multiply(A=Multiply(A=@n29, B=@n30.Result), B=LinearInterpolate(A=@n31, B=@n32, Alpha=@n33))
   Metallic=QualitySwitch(Default=ScalarParameter() {defaultValue=0.55, parameterName=Metallic_Low/Med}, High=@n34, Epic=@n34)
   Roughness=LinearInterpolate(A=@n35, B=Multiply(A=@n35, B=@n36), Alpha=NamedRerouteUsage())
  ) {attributeSetTypes [#1, #2, #3]}
  n38 = FunctionInput() {bUsePreviewValueAsDefault=true, inputName=Mask, inputType=FunctionInput_Scalar, sortPriority=1, previewValue=<1 0 0 1>}
  n39 = SmoothStep(
   Min=ScalarParameter() {defaultValue=0.1, parameterName=SmoothStepMin}
   Max=ScalarParameter() {defaultValue=0.8, parameterName=SmoothStepMax}
   Value=NamedRerouteDeclaration(None=@n38) {name=InputMask, nodeColor=<rgba 0.7578 0 1 1>}
  ) {constMin=0.1}
  n40 = StaticSwitch(
   True=Multiply(A=@n39, B=NamedRerouteUsage())
   False=@n39
   Value=FunctionInput(Preview=StaticBool() {value=true}) {bUsePreviewValueAsDefault=true, inputName=UseCheapSlopeMask, inputType=FunctionInput_StaticBool, sortPriority=2}
  )
  FunctionOutput(None=VirtualTextureFeatureSwitch(No=BlendMaterialAttributes(A=@n1, B=@n10, Alpha=@n11), Yes=BlendMaterialAttributes(A=@n12, B=@n37, Alpha=@n40)))
  NamedRerouteDeclaration(None=ComponentMask(None=WorldPosition().XYZ) {g=true, r=true}) {name=WorldUVs, nodeColor=<rgba 1 0.75 0 1>}
  n41 = CollectionParameter() {
   collection=&FortniteMaterialParameters
   parameterName=SunlightDirectionalVector
  }
  n42 = DotProduct(A=ReflectionVectorWS(CustomWorldNormal=PixelNormalWS()), B=ComponentMask(None=@n41) {b=true, g=true, r=true})
  NamedRerouteDeclaration(None=Power(Base=Saturate(None=@n42), Exp=ScalarParameter() {defaultValue=10, parameterName=ReflectionVectorExponent})) {name=SpecLightMask, nodeColor=<rgba 0 1 0.5313 1>}
  n43 = GetMaterialAttributes(None=FunctionInput() {inputType=FunctionInput_MaterialAttributes}) {attributeGetTypes [#1, #5, #3]}
  NamedRerouteDeclaration(None=@n43.Roughness) {name=RoughnessIn, nodeColor=<rgba 0 0.8594 1 1>}
  NamedRerouteDeclaration(None=@n43.Normal) {name=NormalsIn, nodeColor=<rgba 1 0 0.75 1>}
  NamedRerouteDeclaration(None=@n43.BaseColor) {name=BaseColorIn, nodeColor=<rgba 1 0 0.2109 1>}
  n44 = Constant() {r=1.25}
  n45 = Constant() {r=4}
  n46 = Constant3Vector() {constant=<rgba 0 0 1 0>}
  n47 = VertexNormalWS()
  n48 = LinearInterpolate(A=@n47, B=NamedRerouteUsage(), Alpha=Constant() {r=0.2})
  n49 = QualitySwitch(Default=@n47, Epic=Normalize(VectorInput=Abs(None=@n48)))
  NamedRerouteDeclaration(None=SmoothStep(Value=MaterialFunctionCall(Brightness=@n44, Exponent=@n45, Camera Vector=@n46, Vertex / Pixel Normal=@n49) {materialFunction=&MF_Fresnel_FX}.Fresnel)) {name=CheapSlopeMask, nodeColor=<rgba 0 0.5078 1 1>}
  n50 = Constant3Vector() {constant=<rgba 0 0 1 0>}
  n51 = VertexNormalWS()
  NamedRerouteDeclaration(None=Saturate(None=DotProduct(A=@n50, B=@n51))) {name=Zmask, nodeColor=<rgba 0 1 0.8125 1>}
  NamedRerouteDeclaration(None=FunctionInput(Preview=StaticBoolParameter() {parameterName=MaskReflectionCurveOnZ}) {bUsePreviewValueAsDefault=true, inputName=MaskReflectionCurveOnZ, inputType=FunctionInput_StaticBool, sortPriority=3}) {name=ZmaskBool, nodeColor=<rgba 0.9922 0 1 1>}
  NamedRerouteDeclaration(None=@n43.MaterialAttributes) {name=MA-In, nodeColor=<rgba 0 0.5312 1 1>}
 ]}
