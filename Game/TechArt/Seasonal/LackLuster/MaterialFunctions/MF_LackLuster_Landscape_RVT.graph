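# Material function that restyles landscape material attributes with the seasonal LackLuster chrome look, apparently for RVT.
# Base color is brightened, desaturated, tinted and hue-shifted by world-space noise, while two scales of world-XY swirl normals are added onto the weakened source normal.
# An optional height blend mixes this layer with another material using Mask and SlopeMask, taking the chrome layer's height from a dot product on the swirl normal.
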MaterialFunction /Game/TechArt/Seasonal/LackLuster/MaterialFunctions/MF_LackLuster_Landscape_RVT
graph {nodes [
  n1 = NamedRerouteUsage()
  n2 = FunctionInput() {inputName=HeightIn, inputType=FunctionInput_Scalar, sortPriority=2}
  n3 = TextureSample(UVs=Divide(A=NamedRerouteUsage(), B=ScalarParameter() {defaultValue=15500, group=ChromeRVT, parameterName=HueShiftDivisor}) {constB=2000}) {
   samplerSource=SSM_Wrap_WorldGroupSettings
   samplerType=SAMPLERTYPE_LinearColor
   texture=/Game/Global/Textures/Default/Noise/T_Noise01_Inverse
  }
  n4 = ScalarParameter() {defaultValue=-0.65, group=ChromeRVT, parameterName=WorldHueShiftBias}
  n5 = VectorParameter() {group=ChromeRVT, parameterName=DiffuseSource, defaultValue=<rgba 0.3401 0.2591 0.3455 0>}
  n6 = GetMaterialAttributes(None=FunctionInput() {inputName=MA, inputType=FunctionInput_MaterialAttributes}) {attributeGetTypes [#1, #2]}
  n7 = Add(A=@n6.BaseColor, B=ScalarParameter() {defaultValue=0.1, group=ChromeRVT, parameterName=DiffuseSourceBias})
  n8 = ScalarParameter() {defaultValue=3, group=ChromeRVT, parameterName=DiffuseSourceScale}
  n9 = Desaturation(
   None=Saturate(None=Multiply(A=@n7, B=@n8))
   Fraction=ScalarParameter() {defaultValue=0.95, group=ChromeRVT, parameterName=DiffuseSourceDesaturation}
  )
  n10 = MaterialFunctionCall(
   Hue Shift Percentage=Multiply(A=Add(A=@n3.R, B=@n4), B=ScalarParameter() {defaultValue=2.5, group=ChromeRVT, parameterName=WorldHueShiftScale})
   Texture=Saturate(None=Multiply(A=@n5.RGB, B=@n9))
  ) {materialFunction=&HueShift}
  n11 = VectorParameter() {parameterName=DiffuseColorScale, defaultValue=<rgba 0.9491 1.0755 1.1 0>}
  n12 = NamedRerouteUsage()
  n13 = ScalarParameter() {defaultValue=1.4, parameterName=DiffuseColorMaskExponent}
  n14 = Multiply(A=NamedRerouteUsage(), B=VectorParameter() {group=ChromeRVT, parameterName=SourceNormalsStrength, defaultValue=<rgba 0.3 0.3 1 0>}.RGB)
  n15 = TextureSampleParameter2D(UVs=Divide(A=NamedRerouteUsage(), B=ScalarParameter() {defaultValue=1344, group=ChromeRVT, parameterName=Chrome_Divisor_UV_a})) {
   parameterName=ChromeNormals
   samplerSource=SSM_Wrap_WorldGroupSettings
   samplerType=SAMPLERTYPE_Normal
   texture=/Game/TechArt/Seasonal/LackLuster/Textures/T_Lackluster_Swirl_N
  }
  n16 = TextureSampleParameter2D(UVs=Divide(A=NamedRerouteUsage(), B=ScalarParameter() {defaultValue=2352, group=ChromeRVT, parameterName=Chrome_Divisor_UV_b})) {
   parameterName=ChromeNormals
   samplerSource=SSM_Wrap_WorldGroupSettings
   samplerType=SAMPLERTYPE_Normal
   texture=&T_Lackluster_Swirl_N
  }
  n17 = Normalize(VectorInput=Add(A=@n15.RGB, B=@n16.RGB))
  n18 = Multiply(
   A=Multiply(A=VectorParameter() {group=Normals, parameterName=ChromeNormalScaleOuter, defaultValue=<rgba 0.13 0.13 1 0>}.RGB, B=@n17)
   B=OneMinus(None=NamedRerouteUsage())
  )
  n19 = SetMaterialAttributes!(
   MaterialAttributes=NamedRerouteUsage()
   BaseColor=Multiply(A=Multiply(A=@n10.Result, B=@n11.RGB), B=Power(Base=@n12, Exp=@n13))
   Specular=Constant()
   Normal=Normalize(VectorInput=Add(A=@n14, B=@n18))
  ) {attributeSetTypes [#1, #3, #2]}
  n20 = Constant3Vector() {constant=<rgba 0.5 0.5 0 0>}
  n21 = Saturate(None=ConstantBiasScale(None=DotProduct(A=@n17, B=@n20)))
  n22 = FunctionInput() {bUsePreviewValueAsDefault=true, inputName=Mask, inputType=FunctionInput_Scalar, sortPriority=1, previewValue=<1 0 0 1>}
  n23 = FunctionInput() {bUsePreviewValueAsDefault=true, inputName=SlopeMask, inputType=FunctionInput_Scalar, sortPriority=3, previewValue=<1 0 0 1>}
  n24 = NamedRerouteDeclaration(None=@n23) {name=SlopeMask, nodeColor=<rgba 0.1953 0 1 1>}
  n25 = Saturate(None=Subtract(A=NamedRerouteDeclaration(None=@n22) {name=InputMask, nodeColor=<rgba 0 0.2266 1 1>}, B=OneMinus(None=@n24)))
  n26 = StaticBool() {value=true}
  n27 = Constant()
  n28 = FunctionInput(Preview=StaticBool() {value=true}) {bUsePreviewValueAsDefault=true, inputName=BlendLandscapeLayers, inputType=FunctionInput_StaticBool, sortPriority=4, previewValue=<1 0 0 1>}
  FunctionOutput(None=StaticSwitch(
    True=MaterialFunctionCall!(
     Material 1=@n1
     Height Texture 1=@n2
     Material 2=@n19
     Height Texture 2=@n21
     Transistion Phase=@n25
     UseContrast=@n26
     Contrast=@n27
     AddLayerMask
    ) {materialFunction=/Game/Athena/Apollo/Environments/Landscape/Materials/Biome/CommonMaterials/MF_BlendLandscapeLayers}.BlendedMaterial
    False=@n19
    Value=@n28
   ) {defaultValue=true})
  NamedRerouteDeclaration(None=ComponentMask(None=WorldPosition().XYZ) {g=true, r=true}) {name=WorldUVs, nodeColor=<rgba 0.2891 0 1 1>}
  NamedRerouteDeclaration(None=@n6.MaterialAttributes) {name=MA_In, nodeColor=<rgba 0 1 0.9063 1>}
  NamedRerouteDeclaration(None=@n6.Normal) {name=Normal_In, nodeColor=<rgba 1 0.8438 0 1>}
 ]}
