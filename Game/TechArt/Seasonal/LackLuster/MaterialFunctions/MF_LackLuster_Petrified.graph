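# Material function for the LackLuster season's petrified crystal surface, a metallic, iridescent look with optional blending over incoming material attributes.
# World-aligned masks and normals feed curve-atlas colour, view-angle spectrum reflections and world-noise hue shifts.
# Emissive adds parallax Nanomist triangles, Fresnel and time-of-day scaling.
# The result fades to simplified HLOD shading by distance and, at Epic quality, blends into the landscape's runtime virtual texture by height and slope.
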MaterialFunction /Game/TechArt/Seasonal/LackLuster/MaterialFunctions/MF_LackLuster_Petrified
graph {nodes [
  CameraVectorWS()
  n1 = RuntimeVirtualTextureSample!(UVs, World Position, Mip Bias=ScalarParameter() {defaultValue=4, parameterName=RVTMipLevelBias}) {
   materialType=BaseColor_Normal_Specular_YCoCg
   mipValueMode=RVTMVM_MipBias
   virtualTexture=/Game/Athena/Apollo/Maps/Landscape/VT/RVT_Apollo_Landscape
  }
  n2 = SetMaterialAttributes!(
   MaterialAttributes
   BaseColor=Multiply(A=VectorParameter() {parameterName=RVTBaseColorScale, defaultValue=<rgba 0.4757 1.2628 2 0>}.RGB, B=@n1.BaseColor)
   Specular=@n1.Specular
   Roughness=@n1.Specular
  ) {attributeSetTypes [#1, #2, #3]}
  n3 = SphereMask(
   A=RuntimeVirtualTextureSample!(UVs, World Position) {materialType=WorldHeight, virtualTexture=/Game/Athena/Apollo/Maps/Landscape/VT/RVT_Apollo_Landscape_Height}.WorldHeight
   B=ComponentMask(None=WorldPosition().XYZ) {b=true}
   Radius=ScalarParameter() {defaultValue=95, parameterName=RVTBlendRadius}
   Hardness=ScalarParameter() {parameterName=RVTBlendHardness}
  )
  n4 = Power(Base=@n3, Exp=ScalarParameter() {defaultValue=1, parameterName=RVTBlendExponent})
  n5 = ScalarParameter() {parameterName=RVTBlendBias}
  n6 = Subtract(A=Multiply(A=Add(A=@n4, B=@n5), B=ScalarParameter() {defaultValue=2.2, parameterName=RVTBlendScale}), B=NamedRerouteUsage())
  n7 = DotProduct(A=VertexNormalWS(), B=Constant3Vector() {constant=<rgba 0 0 1 0>})
  n8 = Multiply(A=Multiply(A=@n3, B=Saturate(None=@n6)), B=OneMinus(None=Saturate(None=@n7)))
  n9 = FunctionInput(Preview=SetMaterialAttributes!(MaterialAttributes)) {bUsePreviewValueAsDefault=true, inputName=MA_In, inputType=FunctionInput_MaterialAttributes}
  n10 = SetMaterialAttributes!(
   MaterialAttributes
   BaseColor=NamedRerouteUsage()
   Metallic=NamedRerouteUsage()
   Specular=ScalarParameter() {defaultValue=0.2, parameterName=Specular}
   Roughness=NamedRerouteUsage()
   OpacityMask=NamedRerouteUsage()
   Normal=NamedRerouteUsage()
   AmbientOcclusion=LinearInterpolate(A=ScalarParameter() {defaultValue=0.9, parameterName=AO_Low}, Alpha=NamedRerouteUsage())
   EmissiveColor=NamedRerouteUsage()
   WorldPositionOffset=NamedRerouteUsage()
  ) {attributeSetTypes [
    #1
    #4
    #2
    #3
    #5
    #6
    #7
    #8
    #9
   ]}
  n11 = SetMaterialAttributes(
   BaseColor=LinearInterpolate(
    A=VectorParameter() {parameterName=HLOD_Color1, defaultValue=<rgba 0.227 0.0999 0.1845 1>}.RGB
    B=VectorParameter() {parameterName=HLOD_Color2, defaultValue=<rgba 0.3663 0.2705 0.305 1>}.RGB
    Alpha=NamedRerouteUsage()
   )
   Metallic=ScalarParameter() {defaultValue=0.9, parameterName=HLOD_Metallic}
   Specular=ScalarParameter() {defaultValue=0.5, parameterName=HLOD_Specular}
   Roughness=ScalarParameter() {parameterName=HLOD_Roughness}
   Normal=NamedRerouteUsage()
   AmbientOcclusion=Constant() {r=1}
   EmissiveColor=Constant()
   WorldPositionOffset=NamedRerouteUsage()
  ) {
   desc="HLOD!"
   attributeSetTypes [
    #1
    #4
    #2
    #3
    #5
    #6
    #7
    #8
    #9
   ]
  }
  n12 = MaterialProxyReplace(Realtime=BlendMaterialAttributes(A=@n10, B=@n11, Alpha=NamedRerouteUsage()), MaterialProxy=@n11)
  n13 = StaticBool()
  n14 = StaticSwitch(
   True=BlendMaterialAttributes(
    A=NamedRerouteDeclaration(None=@n9) {name=MA_In, nodeColor=<rgba 1 0.9844 0 1>}
    B=@n12
    Alpha=FunctionInput() {bUsePreviewValueAsDefault=true, inputName=BlendMaterialAttributesAlpha, inputType=FunctionInput_Scalar, sortPriority=1}
   )
   False=@n12
   Value=NamedRerouteDeclaration(None=FunctionInput(Preview=@n13) {bUsePreviewValueAsDefault=true, inputName=bBlendMaterialAttributes, inputType=FunctionInput_StaticBool, sortPriority=-1}) {name=bBlendMaterialAttributes, nodeColor=<rgba 1 0 0.4453 1>}
  )
  n15 = QualitySwitch(
   Default=@n14
   Epic=MaterialFunctionCall(NoRVT=@n14, RVT=BlendMaterialAttributes(A=@n14, B=@n2, Alpha=@n8)) {materialFunction=&FN_RVT_Switch_Material}.Result
  )
  FunctionOutput(None=MaterialProxyReplace(Realtime=StaticSwitchParameter(True=@n15, False=@n14) {defaultValue=true, parameterName=UseRVTBlending}, MaterialProxy=@n14)) {outputName=MA Out}
  n16 = NamedRerouteDeclaration(None=ChannelMaskParameter(None=NamedRerouteUsage()) {maskChannel=Blue, defaultValue=<rgba 0 0 1 0>}.R) {name=MasksWorldAligned.Masked, nodeColor=<rgba 1 0.8906 0 1>}
  n17 = LinearInterpolate(
   A=ScalarParameter() {defaultValue=0.22, parameterName=RoughnessLow}
   B=ScalarParameter() {defaultValue=0.25, parameterName=Roughness}
   Alpha=@n16
  )
  MaterialProxyReplace(Realtime=@n17, MaterialProxy=ScalarParameter() {parameterName=RoughnessHLOD})
  n18 = TextureObjectParameter() {
   parameterName=Normal_WorldAligned
   samplerType=SAMPLERTYPE_Normal
   texture=/Game/TechArt/Seasonal/LackLuster/Textures/T_Lackluster_Petrified_CrystalB_N
  }
  n19 = Divide(A=NamedRerouteUsage(), B=VectorParameter() {parameterName=CrystalWorldTileDivisor, defaultValue=<rgba 10 10 10 1>}.RGB) {constB=100}
  n20 = NamedRerouteDeclaration(None=MaterialFunctionCall(TextureObject=@n18, WorldPosition=NamedRerouteDeclaration(None=@n19) {name=CrystalWorldTileUVs, nodeColor=<rgba 1 0 0.8906 1>}) {materialFunction=&WorldAlignedNormal}.XYZ Texture) {name=WorldAlignedNormalTexture, nodeColor=<rgba 0 1 0.25 1>}
  n21 = Constant3Vector() {constant=<rgba 0 0 1 0>}
  n22 = VertexColor()
  n23 = AppendVector(A=@n22, B=@n22)
  n24 = NamedRerouteDeclaration(None=StaticSwitchParameter(True=ChannelMaskParameter(None=@n23) {parameterName=VertexColorMask}.R, False=Constant()) {defaultValue=true, parameterName=UseVertexColorMask}) {name=VertexColorMask, nodeColor=<rgba 1 0 0.8203 1>}
  n25 = TextureObjectParameter() {
   parameterName=DetailNormal_WorldAligned
   samplerType=SAMPLERTYPE_Normal
   texture=/Game/TechArt/Seasonal/LackLuster/Textures/T_Lackluster_Petrified_Crystal_Detail_N
  }
  n26 = FunctionInput(Preview=WorldPosition() {worldPositionShaderOffset=WPT_ExcludeAllShaderOffsets}.XYZ) {bUsePreviewValueAsDefault=true, description=Defaults to Absolute World Position, inputName=Texture Override Position, sortPriority=400}
  n27 = Divide(
   A=NamedRerouteDeclaration(None=@n26) {name=Texture Override Position, nodeColor=<rgba 1 0.1 0.1 1>}
   B=VectorParameter() {parameterName=DetailNormallWorldTileDivisor, defaultValue=<rgba 2 2 2 1>}.RGB
  ) {constB=100}
  n28 = Add(
   A=Multiply(
    A=LinearInterpolate(A=@n20, B=@n21, Alpha=@n24)
    B=VectorParameter() {parameterName=CrystalNormalStrength, defaultValue=<rgba 0.2 0.2 1 0>}.RGB
   )
   B=Multiply(
    A=MaterialFunctionCall(TextureObject=@n25, WorldPosition=@n27) {materialFunction=&WorldAlignedNormal}.XYZ Texture
    B=VectorParameter() {parameterName=DetailNormalStrength, defaultValue=<rgba 0.2 0.2 0 0>}.RGB
   )
  )
  n29 = Add(
   A=@n28
   B=TextureSampleParameter2D() {parameterName=Normals, samplerType=SAMPLERTYPE_Normal, texture=&FlatNormalBlue}.RGB
  )
  NamedRerouteDeclaration(None=Normalize(VectorInput=StaticSwitchParameter(True=@n29, False=@n28) {parameterName=UseNormalMap})) {name=Normals, nodeColor=<rgba 1 0 0.9141 1>}
  n30 = StaticBool() {value=true}
  n31 = TextureObjectParameter() {parameterName=Masks_WorldAligned, texture=/Game/TechArt/Seasonal/LackLuster/Textures/T_Lackluster_Petrified_CrystalB_Pack}
  n32 = NamedRerouteUsage()
  n33 = MaterialFunctionCall(Float4=NamedRerouteDeclaration(None=MaterialFunctionCall(Export Float 4=@n30, TextureObject=@n31, WorldPosition=@n32) {materialFunction=/Game/Weapons/WeaponSkins/Wraps/MaterialFunctions/WorldAlignedTexture}.XYZ Texture) {name=MasksWorldAligned.RGBA, nodeColor=<rgba 0 1 0.9297 1>}) {materialFunction=&BreakOutFloat4Components}
  NamedRerouteDeclaration(None=@n33.B) {name=NoiseMap, nodeColor=<rgba 0.1016 1 0 1>}
  NamedRerouteDeclaration(None=StaticSwitchParameter(
    True=MaterialFunctionCall!(
     WSDeformationCenter
     Code_DestructionAmount
     Code_UseUvForDamage
     Code_UseHitEffects
     Code_Emissive
     WSDamageRadius
     CrackDamageShader
     Diffuse
     ScalarModDamageCracks
     ScalarOffsetforDamageCracks
     ExposeLerpFunctionForInvertingTheHitDistanceCalc
     InvertDamageFalloff
     CrackWorldPosition
     WorldPosition
     DistanceCullFadeAnimationOffset
    ) {materialFunction=/Game/Packages/jl_tests/Fortnite_HitEffects}.Hit WPO
    False=Constant()
   ) {parameterName=UseHitEffects}) {name=WPO, nodeColor=<rgba 0.6875 0 1 1>}
  n34 = ScalarParameter() {defaultValue=1.7, parameterName=SunMoonFresnelBrightness}
  n35 = ScalarParameter() {defaultValue=2, parameterName=SunMoonFresnelExponent}
  n36 = SkyAtmosphereLightDirection()
  n37 = NamedRerouteUsage()
  NamedRerouteDeclaration(None=VertexInterpolator(VS=MaterialFunctionCall(Brightness=@n34, Exponent=@n35, Camera Vector=@n36, Vertex / Pixel Normal=@n37) {materialFunction=&MF_Fresnel_FX}.Fresnel).PS) {name=SunMoonMask, nodeColor=<rgba 1 0 0.5156 1>}
  n38 = TextureSampleParameter2D() {parameterName=HeightMap, texture=/Game/Creative/BuildingActors/Props/Textures/CP_White_Texutre}
  NamedRerouteDeclaration(None=StaticSwitchParameter(True=Min(A=@n38.R, B=@n33.R), False=@n33.R) {parameterName=UseHeightMap}) {name=HeightMap, nodeColor=<rgba 0.8516 0 1 1>}
  NamedRerouteDeclaration(None=@n33.A) {name=CrystalMap, nodeColor=<rgba 0 1 0.3672 1>}
  n39 = NamedRerouteUsage()
  n40 = NamedRerouteUsage()
  NamedRerouteDeclaration(None=StaticSwitchParameter(
    True=MaterialFunctionCall!(UseLocalSpace, WorldPosition=@n39, NormalOverride=@n40) {materialFunction=/Game/Environments/Apollo/Sets/Skyfire/Materials/MaterialFunctions/MF_WorldAlignedUVs_HardEdge}.Result
    False=TextureCoordinate()
   ) {defaultValue=true, parameterName=UseCheapTriplanarUVs}) {name=UVs, nodeColor=<rgba 0.2891 0 1 1>}
  n41 = NamedRerouteUsage()
  n42 = ScalarParameter() {defaultValue=141, parameterName=NanomistTriangleUVDivisor}
  n43 = TextureSample(UVs=BumpOffset(Coordinate=Divide(A=@n41, B=@n42) {constB=500}, Height=ScalarParameter() {defaultValue=-2, parameterName=NanomistBumpOffsetHeight})) {
   samplerSource=SSM_Wrap_WorldGroupSettings
   samplerType=SAMPLERTYPE_LinearColor
   texture=/Game/TechArt/Seasonal/LackLuster/Textures/T_Lackluster_Triangles
  }
  n44 = ReflectionVectorWS()
  n45 = Multiply(A=Abs(None=ComponentMask(None=@n44) {b=true, r=true}), B=ScalarParameter() {defaultValue=0.3, parameterName=NanomistReflectionVectorScale})
  n46 = TextureSampleParameter2D!(UVs=@n45, Level, Apply View MipBias) {
   constMipValue=5
   group=Chrome_Phased
   mipValueMode=TMVM_MipLevel
   parameterName=PhasedMap
   samplerSource=SSM_Wrap_WorldGroupSettings
   texture=/Game/TechArt/Seasonal/LackLuster/Textures/T_Lackluster_Phased
  }
  n47 = TextureSampleParameter2D(UVs=@n45) {
   constMipValue=5
   group=Chrome_Phased
   mipValueMode=TMVM_MipLevel
   parameterName=PhasedMap
   samplerSource=SSM_Wrap_WorldGroupSettings
   texture=&T_Lackluster_Phased
  }
  n48 = Max(A=@n46.RGB, B=@n47.RGB)
  n49 = Multiply(
   A=@n43.R
   B=Multiply(A=ShadingPathSwitch(Default=@n48, Mobile=@n46.RGB), B=ScalarParameter() {defaultValue=0.2, parameterName=NanomistColorScale})
  )
  n50 = InverseLinearInterpolate(
   A=ScalarParameter() {defaultValue=1000, parameterName=NanomistFadeDistanceStart}
   B=ScalarParameter() {defaultValue=350, parameterName=NanomistFadeDistanceEnd}
   Value=Distance(A=CameraPositionWS(), B=WorldPosition().XYZ)
  ) {constA=500, constB=1500}
  NamedRerouteDeclaration(None=Multiply(A=Multiply(A=@n43.G, B=@n49), B=Saturate(None=@n50))) {name=NanomistTriangles, nodeColor=<rgba 0 0.0859 1 1>}
  n51 = LinearInterpolate(A=@n33.G, Alpha=NamedRerouteUsage())
  n52 = TextureSampleParameter2D!(UVs, Apply View MipBias) {
   desc="This is really AO! Named to help prevent the need for a code fix"
   parameterName=Diffuse
   texture=&CP_White_Texutre
   channelNames {a="", b="", g="", r=Ambient Occlusion}
  }
  NamedRerouteDeclaration(None=StaticSwitchParameter(True=Min(A=@n51, B=@n52.Ambient Occlusion), False=@n51) {parameterName=UseAOMap}) {name=AOMap, nodeColor=<rgba 1 0 0.2109 1>}
  n53 = NamedRerouteUsage()
  n54 = TextureSampleParameter2D() {
   group=OpacityMask
   parameterName=OpacityMask
   samplerSource=SSM_Wrap_WorldGroupSettings
   texture=&CP_White_Texutre
  }
  n55 = StaticSwitchParameter(
   True=Multiply(
    A=GetMaterialAttributes(None=@n53) {attributeGetTypes [#5]}.OpacityMask
    B=ChannelMaskParameter(None=@n54.RGBA) {parameterName=OpacityMaskChannel}.R
   )
   False=Constant() {r=1}
  ) {parameterName=UseOpacityMask}
  n56 = NamedRerouteUsage()
  NamedRerouteDeclaration(None=StaticSwitchParameter(True=Multiply(A=@n55, B=@n56), False=@n55) {parameterName=UsePhasingCameraMask}) {name=OpacityMask, nodeColor=<rgba 0.3828 1 0 1>}
  n57 = Add(A=Distance(A=CameraPositionWS(), B=WorldPosition().XYZ), B=ScalarParameter() {defaultValue=-20000, parameterName=HLODTransition_Offset})
  n58 = ScalarParameter() {defaultValue=10000, parameterName=HLODTransition_Fade}
  NamedRerouteDeclaration(None=Saturate(None=Divide(A=@n57, B=@n58))) {name=HLODTransition_Blend, nodeColor=<rgba 0.7344 0 1 1>}
  n59 = NamedRerouteUsage()
  n60 = NamedRerouteUsage()
  n61 = Multiply(A=NamedRerouteUsage(), B=ScalarParameter() {defaultValue=0.325, parameterName=SSSCoreEmissiveScale})
  n62 = NamedRerouteUsage()
  n63 = ScalarParameter() {defaultValue=-0.075, parameterName=NoiseMapEmissiveScale}
  n64 = Saturate(None=Subtract(A=@n61, B=Multiply(A=@n62, B=@n63)))
  n65 = NamedRerouteUsage()
  n66 = ScalarParameter() {defaultValue=0.2, parameterName=Fraction}
  n67 = NamedRerouteUsage()
  n68 = Desaturation(None=@n65, Fraction=StaticSwitchParameter(True=Multiply(A=@n66, B=@n67), False=@n66) {defaultValue=true, parameterName=UseHeighGradient})
  n69 = ScalarParameter() {defaultValue=0.8, parameterName=Emissive_AO_Low}
  n70 = ScalarParameter() {defaultValue=0.5, parameterName=Emissive_AO_Low_Mobile}
  n71 = Multiply(
   A=Multiply(
    A=StaticSwitchParameter(True=@n64, False=@n61) {defaultValue=true, parameterName=UseNoiseMapEmissive}
    B=StaticSwitchParameter(True=@n68, False=@n65) {defaultValue=true, parameterName=Emissive_DesaturateBaseColor}
   )
   B=LinearInterpolate(A=QualitySwitch(Default=@n69, Low=@n70), Alpha=NamedRerouteUsage()) {constA=0.5}
  )
  n72 = QualitySwitch(Default=StaticSwitchParameter(True=Add(A=@n60, B=@n71), False=@n71) {defaultValue=true, parameterName=UseNanomistTriangles}, Low=@n71)
  n73 = ScalarParameter() {defaultValue=0.1, parameterName=EmissiveFresnelLow}
  n74 = ScalarParameter() {defaultValue=1, parameterName=EmissiveFresnelHigh}
  n75 = MaterialFunctionCall(
   Brightness=ScalarParameter() {defaultValue=1.2, parameterName=EmissiveFresnelBrightness}
   Exponent=ScalarParameter() {defaultValue=3, parameterName=EmissiveFresnelExponent}
  ) {materialFunction=&MF_Fresnel_FX}
  n76 = Multiply(
   A=Multiply(
    A=MaterialFunctionCall(Hue Shift Percentage=@n59, Texture=@n72) {materialFunction=&HueShift}.Result
    B=LinearInterpolate(A=@n73, B=@n74, Alpha=@n75.Fresnel)
   )
   B=NamedRerouteUsage()
  )
  n77 = TextureCoordinate()
  n78 = OneMinus(None=SphereMask(
    A=ComponentMask(None=@n77) {r=true}
    B=Constant() {r=0.5}
    Radius=ScalarParameter() {defaultValue=0.5, parameterName=EmissiveEdgeRadius}
    Hardness=ScalarParameter() {defaultValue=0.85, parameterName=EmissiveEdgeHardness}
   ))
  n79 = Multiply(A=Multiply(A=@n78, B=@n78), B=Min(A=NamedRerouteUsage(), B=NamedRerouteUsage()))
  n80 = Power(Base=@n78, Exp=ScalarParameter() {defaultValue=8, parameterName=EmissiveEdgeExponent})
  n81 = NamedRerouteDeclaration(None=Multiply(A=Add(A=@n79, B=@n80), B=ScalarParameter() {defaultValue=1.5, parameterName=EmissiveEdgeScale})) {name=EmissiveEdgeMask, nodeColor=<rgba 1 0.7734 0 1>}
  n82 = StaticSwitchParameter(True=Add(A=@n76, B=Multiply(A=@n76, B=@n81)), False=@n76) {defaultValue=true, parameterName=UseEmissiveEdgeMask}
  n83 = OneMinus(None=Multiply(A=NamedRerouteUsage(), B=ScalarParameter() {defaultValue=0.9, parameterName=EmissiveVertexColorScale}))
  n84 = ScalarParameter() {defaultValue=0.7, parameterName=EmissiveScaleMorning}
  n85 = ScalarParameter() {defaultValue=1, parameterName=EmissiveScaleDay}
  n86 = ScalarParameter() {defaultValue=0.42, parameterName=EmissiveScaleEvening}
  n87 = ScalarParameter() {defaultValue=0.65, parameterName=EmissiveScaleNight}
  NamedRerouteDeclaration(None=Multiply(
    A=Multiply(A=@n82, B=@n83)
    B=MaterialFunctionCall(Value_Morning=@n84, Value_Day=@n85, Value_Evening=@n86, Value_Night=@n87) {materialFunction=&Scalar_Time_of_Day}.Result
   )) {name=Emissive, nodeColor=<rgba 1 0 0.8203 1>}
  n88 = SphereMask(
   A=MaterialFunctionCall(Float2=TextureCoordinate()) {materialFunction=&BreakOutFloat2Components}.R
   B=Constant() {r=0.5}
   Radius=ScalarParameter() {defaultValue=0.5, parameterName=CoreGradientRadius}
   Hardness=ScalarParameter() {group=Core, parameterName=CoreGradientHardness}
  )
  n89 = StaticSwitchParameter(
   True=LinearInterpolate(
    A=ScalarParameter() {defaultValue=0.8, group=Core, parameterName=CoreGradientLow}
    Alpha=NamedRerouteDeclaration(None=@n88) {name=CoreGradientMask, nodeColor=<rgba 0.0078 1 0 1>}
   ) {constA=0.7}
   False=Constant() {r=1}
  ) {defaultValue=true, group=Core, parameterName=UseCoreGradientMask}
  n90 = Normalize(VectorInput=Subtract(A=CameraPositionWS(), B=WorldPosition().XYZ))
  n91 = DotProduct(A=@n90, B=Constant3Vector() {constant=<rgba 1 0 0 0>})
  n92 = DotProduct(A=@n90, B=Constant3Vector() {constant=<rgba 0 1 0 0>})
  n93 = Add(A=Divide(A=Arctangent2Fast(Y=@n91, X=@n92)) {constB=6.282}) {constB=0.5}
  n94 = DotProduct(A=NamedRerouteUsage(), B=VectorParameter() {parameterName=ReflectionCurveNormalDistortionScale, defaultValue=<rgba 1 1 0 1>}.RGB)
  n95 = CurveAtlasRowParameter(CurveTime=Frac(None=Add(A=@n93, B=@n94) {constB=0})) {
   atlas=/Game/TechArt/Seasonal/LackLuster/Curves/CA_Lackluster
   curve=/Game/TechArt/Seasonal/LackLuster/Curves/C_LackLuster_Spectrum_01
   defaultValue=1
   parameterName=ReflectionCurve
  }
  n96 = ScalarParameter() {defaultValue=0.5, parameterName=ReflectionCurveScale}
  n97 = ScalarParameter() {defaultValue=0.5, parameterName=ReflectionCurveHeightMapLow}
  n98 = ScalarParameter() {defaultValue=1, parameterName=ReflectionCurveHeightMapHigh}
  n99 = NamedRerouteUsage()
  n100 = MaterialFunctionCall(Camera Vector=@n90, Vertex / Pixel Normal=QualitySwitch(Default=PixelNormalWS(), Low=VertexNormalWS())) {materialFunction=&MF_Fresnel_FX}
  n101 = VectorParameter() {group=HueShift, parameterName=HueShiftColor, defaultValue=<rgba 0.2156 0.4172 0.4722 0>}
  n102 = Add(
   A=Multiply(A=Multiply(A=@n95, B=@n96), B=LinearInterpolate(A=@n97, B=@n98, Alpha=@n99))
   B=Multiply(
    A=@n100.Fresnel Inverted
    B=MaterialFunctionCall(Hue Shift Percentage=@n100.Fresnel Unclamped, Texture=@n101.RGB) {materialFunction=&HueShift}.Result
   )
  )
  n103 = QualitySwitch(Default=@n102, Low=VertexInterpolator(VS=@n102).PS)
  n104 = CurveAtlasRowParameter(CurveTime=NamedRerouteUsage()) {
   atlas=&CA_Lackluster
   curve=/Game/TechArt/Seasonal/LackLuster/Curves/C_LackLuster_PetrifiedDiffuse_01
   defaultValue=5
   parameterName=DiffuseCurve
  }
  n105 = LinearInterpolate(A=@n104, B=Multiply(A=@n104, B=ScalarParameter() {defaultValue=1.5, parameterName=VertexColorlDiffuseScale}), Alpha=NamedRerouteUsage())
  n106 = Multiply(
   A=Desaturation(None=@n105, Fraction=ScalarParameter() {defaultValue=0.3, parameterName=SunMoonDiffuseDesaturation})
   B=VectorParameter() {parameterName=SunMoonDiffuseScale, defaultValue=<rgba 0.3601 0.5907 0.9514 0>}.RGB
  )
  n107 = NamedRerouteUsage()
  n108 = FunctionInput(Preview=StaticSwitchParameter(True=LinearInterpolate(A=@n106, B=@n105, Alpha=@n107), False=@n105) {parameterName=ModulateDiffuseBySunMoonVector}) {bUsePreviewValueAsDefault=true, inputName=BaseColorOverride, sortPriority=2}
  n109 = MaterialFunctionCall(
   Hue Shift Percentage=NamedRerouteUsage()
   Texture=Multiply(A=NamedRerouteDeclaration(None=@n89) {name=CoreGradientMaskRemapped, nodeColor=<rgba 0 0.2266 1 1>}, B=Add(A=@n103, B=@n108))
  ) {materialFunction=&HueShift}
  n110 = LinearInterpolate(
   A=ScalarParameter() {defaultValue=0.5, parameterName=Z_Gradient_Low}
   Alpha=NamedRerouteDeclaration(None=MaterialFunctionCall() {materialFunction=&BoundingBoxBased_0-1_UVW}.B) {name=Z_Gradient, nodeColor=<rgba 0.3828 0 1 1>}
  ) {constA=0.3}
  n111 = StaticSwitchParameter(
   True=Multiply(A=@n109.Result, B=NamedRerouteDeclaration(None=@n110) {name=Z_Gradient_Remapped, nodeColor=<rgba 0.3828 0 1 1>})
   False=@n109.Result
  ) {defaultValue=true, parameterName=UseHeighGradient}
  n112 = MaterialFunctionCall(
   Value_Morning=ScalarParameter() {defaultValue=1.4, parameterName=BaseColor_Scale_Mobile_Morning}
   Value_Day=ScalarParameter() {defaultValue=1, parameterName=BaseColor_Scale_Mobile_Day}
   Value_Evening=ScalarParameter() {defaultValue=0.25, parameterName=BaseColor_Scale_Mobile_Evening}
   Value_Night=ScalarParameter() {defaultValue=1.8, parameterName=BaseColor_Scale_Mobile_Night}
  ) {materialFunction=&Scalar_Time_of_Day}
  NamedRerouteDeclaration(None=ShadingPathSwitch(Default=@n111, Mobile=Multiply(A=@n111, B=@n112.Result))) {name=BaseColor, nodeColor=<rgba 0 0.6484 1 1>}
  NamedRerouteDeclaration(None=@n17) {name=Roughness, nodeColor=<rgba 1 0.3984 0 1>}
  n113 = CameraPositionWS()
  n114 = WorldPosition()
  n115 = ScalarParameter() {defaultValue=250, parameterName=PhasingCameraMaskRadius}
  n116 = ScalarParameter() {defaultValue=0.8, parameterName=PhasingCameraMaskHardness}
  NamedRerouteDeclaration(None=OneMinus(None=SphereMask(A=@n113, B=@n114.XYZ, Radius=@n115, Hardness=@n116))) {name=PhasingCameraMask, nodeColor=<rgba 0.4062 0 1 1>}
  NamedRerouteDeclaration(None=LinearInterpolate(
    A=ScalarParameter() {defaultValue=0.85, parameterName=MetallicLow}
    B=ScalarParameter() {defaultValue=0.9, parameterName=Metallic}
    Alpha=@n16
   )) {name=Metallic, nodeColor=<rgba 0.5 1 0 1>}
  n117 = ComponentMask(None=Divide(A=NamedRerouteUsage(), B=ScalarParameter() {defaultValue=9500, parameterName=HueShiftDivisor}) {constB=2000}) {g=true, r=true}
  n118 = Multiply(
   A=Add(
    A=TextureSample(UVs=@n117) {samplerType=SAMPLERTYPE_LinearColor, texture=/Game/Global/Textures/Default/Noise/T_Noise01_Inverse}.R
    B=ScalarParameter() {defaultValue=0.65, desc=-0.65, parameterName=HueShiftBias}
   )
   B=ScalarParameter() {defaultValue=0.4, parameterName=WorldHueShiftScale}
  )
  n119 = NamedRerouteUsage()
  n120 = Multiply(
   A=Divide(
    A=MaterialFunctionCall(Float3=@n119) {materialFunction=&BreakOutFloat3Components}.B
    B=ScalarParameter() {defaultValue=2200, parameterName=HueShiftZDivisor}
   )
   B=ScalarParameter() {defaultValue=0.35, desc=0.2, parameterName=HueShiftZScale}
  )
  NamedRerouteDeclaration(None=VertexInterpolator(VS=Add(A=@n118, B=@n120)).PS) {name=WorldHueShiftMask, nodeColor=<rgba 0 1 0.625 1>}
  NamedRerouteDeclaration(None=FunctionInput(Preview=VertexNormalWS()) {bUsePreviewValueAsDefault=true, inputName=VertexNormalOverride, sortPriority=4}) {name=VertexNormal, nodeColor=<rgba 0.5 0 1 1>}
 ]}
